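# Sound cue for first-person sprint footsteps on sand in Athena, built from sand walk samples.
# A quality-level node sends its first two slots to a random pick of eight waves and its last two to a four-wave subset.
# A root modulator varies volume and raises pitch on each step.
# It uses 1P sprint attenuation and local footstep concurrency, with overrides for one instance per owner, stopping the oldest.
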SoundCue /Game/Athena/Sounds/Character/Footsteps/1P/Sand/Athena_Character_Sprint_Footstep_Sand_Cue
root=Modulator_2
nodes [
 {id=Modulator_2, children [QualityLevel_0], properties {pitchMax=1.5, pitchMin=1.3, volumeMax=1, volumeMin=0.7}}
 {id=QualityLevel_0, children [Random_1,Random_1,Random_2,Random_2]}
 {
  id=Random_1
  children [WavePlayer_3, WavePlayer_4, WavePlayer_5, WavePlayer_6, WavePlayer_7, WavePlayer_8, WavePlayer_9, WavePlayer_10]
  properties {weights [1,1,1,1,1,1,1,1]}
 }
 {
  id=Random_2
  children [WavePlayer_11, WavePlayer_12, WavePlayer_13, WavePlayer_14]
  properties {bShouldExcludeFromBranchCulling=true, weights [1,1,1,1]}
 }
 {id=WavePlayer_10, properties {soundWaveAssetPtr=&/FS_Sand_Walk_01}}
 {id=WavePlayer_11, properties {soundWaveAssetPtr=&/FS_Sand_Walk_04}}
 {id=WavePlayer_12, properties {soundWaveAssetPtr=&/FS_Sand_Walk_03}}
 {id=WavePlayer_13, properties {soundWaveAssetPtr=&/FS_Sand_Walk_02}}
 {id=WavePlayer_14, properties {soundWaveAssetPtr=&FS_Sand_Walk_01}}
 {id=WavePlayer_3, properties {soundWaveAssetPtr=&/FS_Sand_Walk_08}}
 {id=WavePlayer_4, properties {soundWaveAssetPtr=&/FS_Sand_Walk_07}}
 {id=WavePlayer_5, properties {soundWaveAssetPtr=&/FS_Sand_Walk_06}}
 {id=WavePlayer_6, properties {soundWaveAssetPtr=&/FS_Sand_Walk_05}}
 {id=WavePlayer_7, properties {soundWaveAssetPtr=&FS_Sand_Walk_04}}
 {id=WavePlayer_8, properties {soundWaveAssetPtr=&FS_Sand_Walk_03}}
 {id=WavePlayer_9, properties {soundWaveAssetPtr=&FS_Sand_Walk_02}}
]
properties {
 attenuationSettings=/Game/Sounds/Attenuation/Fort_Player_Footsteps_1P_Sprint_Athena_Att
 maxDistance=2097152
 soundClassObject=&Fort_Player_Foley_1P_Athena
 volumeMultiplier=1
 concurrencyOverrides {bLimitToOwner=true, maxCount=1, resolutionRule=StopOldest}
 concurrencySet [&Athena_Footstep_Local_Conc]
}
